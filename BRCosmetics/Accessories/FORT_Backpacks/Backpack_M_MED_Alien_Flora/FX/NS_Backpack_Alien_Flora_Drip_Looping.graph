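# Looping Niagara drip effect for the Alien Flora backpack, hanging swaying mesh strands from sampled skeletal mesh points.
# DroolPendulum spawns strands sparsely, resampling the mesh each update as the pivot of a lengthening pendulum pulled downward.
# Each strand orients and stretches between pivot and tip, passes both positions to its material, and fades before dying.
# A simpler Pendulum emitter is disabled, and User.ReactiveRate drives nothing shown.
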NiagaraSystem /BRCosmetics/Accessories/FORT_Backpacks/Backpack_M_MED_Alien_Flora/FX/NS_Backpack_Alien_Flora_Drip_Looping
dependencies {
 usedDataInterfaces [&NiagaraDataInterfaceCurve, &NiagaraDataInterfaceColorCurve]
 usedDynamicInputs []
 usedModules [&SystemState]
}
emitters {
 DroolPendulum {
  emitterUpdateScript [
   &EmitterState(Life Cycle Mode=::NewEnumerator1, Loop Duration=1, UseLoopDelay=false)
   &SpawnRate(
    SpawnRate=&RandomRangeFloat(Minimum=0.2, Maximum=0.5, Evaluation Type=::NewEnumerator1)
    Use Spawn Probability=true
   )
  ]
  particleSpawnScript [
   &TorusLocation(
    Align Torus With Axis=false
    Handle Radius=0.01
    Large Radius=1
    MaskSpawnGroup=false
    Module.Override Randomness=false
    Torus Origin:NiagaraPosition=<0 0 0>
   ) {enabled=false}
   /SetVariables(Particles.TorusPosition:NiagaraPosition=$Particles.Position) {enabled=false, isSetParameters=true}
   /Niagara/Modules/Spawn/Initialization/InitializeParticle(
    Lifetime=&UniformRangedFloat(Minimum=1.2, Maximum=1.5)
    Mesh Scale=<0.1 0.1 0.018>
    Module.Write Color=true
    Module.Write Lifetime=true
    Module.Write Mass=true
    Module.Write Material Random=true
    Module.Write Position=false
    Module.Write Scale=true
    Module.Write SpriteRotation=false
    Module.Write SpriteSize=true
    Sprite Size=&Vector2DFromFloat(Value=2)
   )
   &SkeletalMeshLocation(
    Inherit Velocity=false
    MaskSpawnGroup=false
    Mesh Sampling Type=::NewEnumerator5
    Sampled Position Offset Coordinate Space:ENiagaraMeshTransforms=::NewEnumerator3
    Skeletal Mesh=$User.SkeletalMesh
    Socket Sampling Mode=::NewEnumerator5
   ) {module=SkeletalMeshLocation002}
  ]
  particleUpdateScript [
   &ParticleState()
   /SetVariables(Particles.OrigPos:NiagaraPosition=$Particles.Position) {isSetParameters=true, module=SetVariables_#1}
   &SkeletalMeshLocation(
    Inherit Velocity=false
    MaskSpawnGroup=false
    Mesh Sampling Type=::NewEnumerator5
    Sampled Position Offset Coordinate Space:ENiagaraMeshTransforms=::NewEnumerator3
    Skeletal Mesh=$User.SkeletalMesh
    Socket Sampling Mode=::NewEnumerator5
   )
   /SetVariables(Particles.MeshPosition:NiagaraPosition=$Particles.Position) {isSetParameters=true, module=SetVariables_#2}
   /SetVariables(Particles.Position:Vector3f=$Particles.OrigPos) {isSetParameters=true, module=SetVariables_#3}
   /SetVariables(
    Particles.XYScale:NiagaraFloat=&FloatFromCurve(
     FloatCurve=DataInterface() {propertyValues {Curve {keys [
         {arriveTangent=5.3519, interpMode=RCIM_Cubic, leaveTangent=5.3519, tangentMode=RCTM_User, time=0, value=0}
         {arriveTangent=-0.99, interpMode=RCIM_Cubic, leaveTangent=-0.99, tangentMode=RCTM_User, time=0.25, value=1}
         {arriveTangent=-0.8998, interpMode=RCIM_Cubic, leaveTangent=-0.8998, tangentMode=RCTM_User, time=0.6, value=0.6082}
         {arriveTangent=-0.6378, interpMode=RCIM_Cubic, leaveTangent=-0.6378, tangentMode=RCTM_User, time=1, value=0.25}
        ]}}}
     Scale Curve=&RandomRangeFloat(Minimum=0.6, Maximum=0.8)
    )
    Particles.ZScale:NiagaraFloat=&FloatFromCurve(
     FloatCurve=DataInterface() {propertyValues {Curve {keys [
         {arriveTangent=3.0916, interpMode=RCIM_Cubic, leaveTangent=3.0916, tangentMode=RCTM_User, time=0, value=0.1}
         {arriveTangent=0.4235, interpMode=RCIM_Cubic, leaveTangent=0.4235, tangentMode=RCTM_User, time=0.6784, value=0.9948}
         {arriveTangent=0.1347, interpMode=RCIM_Cubic, leaveTangent=0.1347, tangentMode=RCTM_User, time=0.9645, value=1.2125}
         <at 0.9984 1.5>
        ]}}}
     Scale Curve=&RandomRangeFloat(Minimum=5, Maximum=7)
    )
   ) {isSetParameters=true, module=SetVariables_#4}
   &PendulumSetup(
    Module.Enable Spring Driver=false
    Pendulum Length:NiagaraFloat=$Particles.ZScale
    Pendulum Pivot:Vector3f=$Particles.MeshPosition
    Rigid Pendulum=false
   )
   &Drag(Drag=-10, Rotational Drag=-10, Use Linear Drag=true, Use Rotational Drag=true)
   &AddVelocity(Coordinate Space=::World, Velocity=<0 0 -40>)
   &AddVelocity(Velocity=&Lerp_Vector(A=<0 0 0>, B=<100 0 50>, Alpha=$System.RoarCurve)) {enabled=false, module=AddVelocity001}
   &InheritVelocity(Module.Limit Speed=true, Velocity Scale=&Lerp_Vector(A=<-1 -1 -1>, B=<0 0 0>, Alpha=$Particles.NormalizedAge))
   &SolveForcesAndVelocity(Clamp Velocity=false, Limit Acceleration=false)
   &PendulumConstraint()
   &CalculateAccurateVelocity()
   /SetVariables(
    Particles.MeshFacing:Vector3f=&NormalizeVector(Vector To Normalize=&Subtract_Vector(A=$Particles.MeshPosition, B=Linked() {linkedVariable {name=Particles.Position, type=NiagaraPosition}}))
    Particles.RestDirection:Vector3f=/Niagara/DynamicInputs/Helpers/SafeNormalizeVector(Vector To Normalize:Vector3f=$Engine.Owner.Velocity, Fallback Vector:Vector3f=<0 0 0>)
   ) {isSetParameters=true, module=SetVariables_#5}
   /Niagara/Modules/Update/Orientation/V2/OrientMeshToVector(
    Look At Direction:Vector3f=&Multiply_VectorByFloat(Vector=$Particles.MeshFacing, Float=1)
    Override Quat=false
    Reference Axis Vector:Vector3f=<0 0 1>
    UseOptionalSideAxis=false
    UseOptionalUpAxis=false
   )
   &DynamicMaterialParameters(
    Index 2 Param 1=HlslExpression() {hlslExpression=Particles.MeshPosition.x}
    Index 2 Param 2=HlslExpression() {hlslExpression=Particles.MeshPosition.y}
    Index 2 Param 3=HlslExpression() {hlslExpression=Particles.MeshPosition.z}
    Index 2 Param 4=&RandomRangeFloat(Minimum=2, Maximum=5)
    Index 3 Param 1=&MakeFloatFromVector(Vector=Linked() {linkedVariable {name=Particles.Position, type=NiagaraPosition}}, Channel=::NewEnumerator0)
    Index 3 Param 2=&MakeFloatFromVector(Vector=Linked() {linkedVariable {name=Particles.Position, type=NiagaraPosition}}, Channel=::NewEnumerator1)
    Index 3 Param 3=&MakeFloatFromVector(Vector=Linked() {linkedVariable {name=Particles.Position, type=NiagaraPosition}}, Channel=::NewEnumerator2)
    Index 3 Param 4=0.025
    Module.Param0WriteEnabled2=true
    Module.Param0WriteEnabled3=true
    Module.Param1WriteEnabled2=true
    Module.Param1WriteEnabled3=true
    Module.Param2WriteEnabled2=true
    Module.Param2WriteEnabled3=true
    Module.Param3WriteEnabled2=true
    Module.Param3WriteEnabled3=true
    Write Parameter Index 0=false
    Write Parameter Index 2=true
    Write Parameter Index 3=true
   )
   &ScaleMeshSize(Scale Factor=&MakeVector(
     X=$Particles.XYScale
     Y=$Particles.XYScale
     Z=&VectorLength(Vector=&Subtract_Vector(A=$Particles.MeshPosition, B=Linked() {linkedVariable {name=Particles.Position, type=NiagaraPosition}}))
    ))
   &Color(Color=&ColorFromCurve(
     ColorCurve=DataInterface() {propertyValues {
       AlphaCurve {keys [<at 0 20 RCIM_Cubic>]}
       BlueCurve {keys [<at 0 0 RCIM_Cubic>,<at 0.5257 0>,<at 1 5 RCIM_Cubic>]}
       GreenCurve {keys [<at 0 0 RCIM_Cubic>,<at 0.5257 5>,<at 1 0 RCIM_Cubic>]}
       RedCurve {keys [<at 0 5 RCIM_Cubic>,<at 0.5257 0>,<at 1 0 RCIM_Cubic>]}
      }}
     CurveIndex=&Divide_Float(A=&MakeFloatFromInt(Int=$Particles.PickBone), B=366)
    ))
   &ScaleColor(
    Scale Alpha=&FloatFromCurve(
     FloatCurve=DataInterface() {propertyValues {Curve {keys [
         {arriveTangent=-0.0275, interpMode=RCIM_Cubic, leaveTangent=-0.0275, tangentMode=RCTM_User, time=0.7, value=1}
         {arriveTangent=-1.0663, interpMode=RCIM_Cubic, leaveTangent=-1.0663, tangentMode=RCTM_User, time=1, value=0}
        ]}}}
     Scale Curve=2
    )
    ScaleA=true
    ScaleRGB=false
   )
   &GenerateDeathEvent()
  ]
  rendererClasses [&NiagaraMeshRendererProperties]
  renderers [{rendererClass=&NiagaraMeshRendererProperties, rendererIndex=0}]
 }
 Pendulum {
  bEnabled=false
  emitterUpdateScript [
   &EmitterState(Life Cycle Mode=::NewEnumerator1, Loop Duration=1, UseLoopDelay=false)
   &SpawnRate(
    SpawnRate=&RandomRangeFloat(Minimum=1, Maximum=1, Evaluation Type=::NewEnumerator1)
    Use Spawn Probability=false
   )
  ]
  particleSpawnScript [
   /Niagara/Modules/Spawn/Initialization/V2/InitializeParticle(
    Color=&RandomRangeLinearColor(Minimum=<rgba 1 1 1 1>, Maximum=<rgba 1 1 1 1>)
    Lifetime Max=1.5
    Lifetime Min=1.2
    Lifetime Mode=::NewEnumerator1
    Mesh Scale=<0.1 0.1 0.018>
    Mesh Scale Mode=::NewEnumerator4
    UsePositionOffset=false
    Write Material Random=true
   ) {module=InitializeParticle001}
   &AddVelocity(Coordinate Space=::World, Velocity=<0 0 -100>) {enabled=false}
  ]
  particleUpdateScript [
   &ParticleState()
   /SetVariables(Particles.OrigPos:NiagaraPosition=$Particles.Position) {isSetParameters=true, module=SetVariables_#1}
   &SkeletalMeshLocation(
    MaskSpawnGroup=false
    Mesh Sampling Type=::NewEnumerator5
    Orientation Sampling=::NewEnumerator1
    Skeletal Mesh=$User.SkeletalMesh
    Socket Sampling Mode=::NewEnumerator5
   ) {module=SkeletalMeshLocation001}
   /SetVariables(Particles.MeshPosition:NiagaraPosition=$Particles.Position) {isSetParameters=true, module=SetVariables_#2}
   /SetVariables(Particles.Position:Vector3f=$Particles.OrigPos) {isSetParameters=true, module=SetVariables_#3}
   /SetVariables(
    Particles.XYScale:NiagaraFloat=&FloatFromCurve(
     FloatCurve=DataInterface() {propertyValues {Curve {keys [
         {arriveTangent=5.3518, interpMode=RCIM_Cubic, leaveTangent=5.3518, tangentMode=RCTM_User, time=0, value=0}
         {arriveTangent=-0.99, interpMode=RCIM_Cubic, leaveTangent=-0.99, tangentMode=RCTM_User, time=0.25, value=1}
         {arriveTangent=-0.8998, interpMode=RCIM_Cubic, leaveTangent=-0.8998, tangentMode=RCTM_User, time=0.6, value=0.6082}
         {arriveTangent=-0.6378, interpMode=RCIM_Cubic, leaveTangent=-0.6378, tangentMode=RCTM_User, time=1, value=0.25}
        ]}}}
     Scale Curve=&RandomRangeFloat(Minimum=0.6, Maximum=0.75)
    )
    Particles.ZScale:NiagaraFloat=&FloatFromCurve(
     FloatCurve=DataInterface() {propertyValues {Curve {keys [
         {arriveTangent=3.0916, interpMode=RCIM_Cubic, leaveTangent=3.0916, tangentMode=RCTM_User, time=0, value=0.1}
         {arriveTangent=0.4235, interpMode=RCIM_Cubic, leaveTangent=0.4235, tangentMode=RCTM_User, time=0.6784, value=0.9948}
         {arriveTangent=0.1347, interpMode=RCIM_Cubic, leaveTangent=0.1347, tangentMode=RCTM_User, time=0.9645, value=1.2125}
         <at 0.9984 1.5>
        ]}}}
     Scale Curve=&RandomRangeFloat(Minimum=5, Maximum=7)
    )
   ) {isSetParameters=true, module=SetVariables_#4}
   &PendulumSetup(
    Module.Enable Spring Driver=false
    Pendulum Length:NiagaraFloat=$Particles.ZScale
    Pendulum Pivot:Vector3f=$Particles.MeshPosition
    Rigid Pendulum=false
   )
   /SetVariables(Particles.Velocity=<0 0 -50>) {isSetParameters=true}
   &InheritVelocity(Module.Limit Speed=true)
   &Drag(Drag=0, Rotational Drag=0, Use Linear Drag=false, Use Rotational Drag=false)
   &SolveForcesAndVelocity(Clamp Velocity=false, Limit Acceleration=false)
   &PendulumConstraint()
   &OrientMeshToVector(
    Look At Direction:Vector3f=&Multiply_VectorByFloat(Vector=$Particles.MeshFacing, Float=1)
    Override Quat=false
    Reference Axis Vector:Vector3f=<0 0 1>
    UseOptionalSideAxis=true
    UseOptionalUpAxis=true
   )
   &CalculateAccurateVelocity()
   /SetVariables(
    Particles.MeshFacing:Vector3f=&NormalizeVector(Vector To Normalize=&Subtract_Vector(A=$Particles.MeshPosition, B=Linked() {linkedVariable {name=Particles.Position, type=NiagaraPosition}}))
    Particles.RestDirection:Vector3f=&SafeNormalizeVector(Vector To Normalize:Vector3f=$Engine.Owner.Velocity, Fallback Vector:Vector3f=<0 0 0>)
   ) {isSetParameters=true, module=SetVariables_#5}
   &DynamicMaterialParameters(
    Index 2 Param 1=HlslExpression() {hlslExpression=Particles.MeshPosition.x}
    Index 2 Param 2=HlslExpression() {hlslExpression=Particles.MeshPosition.y}
    Index 2 Param 3=HlslExpression() {hlslExpression=Particles.MeshPosition.z}
    Index 2 Param 4=&RandomRangeFloat(Minimum=2, Maximum=3)
    Index 3 Param 1=&MakeFloatFromVector(Vector=Linked() {linkedVariable {name=Particles.Position, type=NiagaraPosition}}, Channel=::NewEnumerator0)
    Index 3 Param 2=&MakeFloatFromVector(Vector=Linked() {linkedVariable {name=Particles.Position, type=NiagaraPosition}}, Channel=::NewEnumerator1)
    Index 3 Param 3=&MakeFloatFromVector(Vector=Linked() {linkedVariable {name=Particles.Position, type=NiagaraPosition}}, Channel=::NewEnumerator2)
    Index 3 Param 4=0.025
    Module.Param0WriteEnabled2=true
    Module.Param0WriteEnabled3=true
    Module.Param1WriteEnabled2=true
    Module.Param1WriteEnabled3=true
    Module.Param2WriteEnabled2=true
    Module.Param2WriteEnabled3=true
    Module.Param3WriteEnabled2=true
    Module.Param3WriteEnabled3=true
    Write Parameter Index 0=false
    Write Parameter Index 2=true
    Write Parameter Index 3=true
   )
   &ScaleMeshSize(Scale Factor=&MakeVector(
     X=$Particles.XYScale
     Y=$Particles.XYScale
     Z=&VectorLength(Vector=&Subtract_Vector(A=$Particles.MeshPosition, B=Linked() {linkedVariable {name=Particles.Position, type=NiagaraPosition}}))
    ))
   &ScaleColor(
    Scale Alpha=&FloatFromCurve(
     FloatCurve=DataInterface() {propertyValues {Curve {keys [
         {arriveTangent=-0.0275, interpMode=RCIM_Cubic, leaveTangent=-0.0275, tangentMode=RCTM_User, time=0.7, value=1}
         {arriveTangent=-1.0663, interpMode=RCIM_Cubic, leaveTangent=-1.0663, tangentMode=RCTM_User, time=1, value=0}
        ]}}}
     Scale Curve=2
    )
    ScaleA=true
    ScaleRGB=false
   )
  ]
  rendererClasses [&NiagaraMeshRendererProperties]
  renderers [{rendererClass=&NiagaraMeshRendererProperties, rendererIndex=0}]
 }
}
summary {
 systemName=NS_Backpack_Alien_Flora_Drip_Looping
 emitters [
  {bEnabled=false, emitterName=Pendulum, rendererClasses [&NiagaraMeshRendererProperties]}
  {emitterName=DroolPendulum, rendererClasses [&NiagaraMeshRendererProperties]}
 ]
 userVariables [
  User.ReactiveRate {type=NiagaraFloat, defaultValue=0, description=""}
  User.SkeletalMesh {
   type=NiagaraDataInterfaceSkeletalMesh
   defaultValue {
    struct=NiagaraExt_VariableValue_DataInterface
    value {dataInterface=&self:NiagaraDataInterfaceSkeletalMesh_0}
   }
  }
 ]
}
