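# Anim Blueprint for the Alien Flora cosmetic backpack, playing random idle or bite-break animations and bending its tail during skydiving.
# Chained additive bone-space Modify Bone nodes yaw tail_base and tail_01–13, blending in via the skydiving bool, except always-enabled tail_05; tail_07 is applied three times.
# The update event sets that bool and shared rotator when floating or diving; otherwise it only zeroes the rotator, never clearing the bool.
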AnimBlueprint /BRCosmetics/Accessories/FORT_Backpacks/Backpack_M_MED_Alien_Flora/Meshes/M_MED_Alien_Flora_Pack_AnimBP
target_skeleton=&/M_MED_Alien_Flora_Pack_Skeleton
graphs [
 AnimGraph {nodes [
   Local Space Ref Pose()
   n1 = Random Sequence Player!() -> (Pose:Pose Link Structure) {
    is=title AnimGraphNode_RandomPlayer
    entries [
     {
      chanceToPlay=0.7
      maxLoopCount=0
      maxPlayRate=1
      minLoopCount=0
      minPlayRate=1
      sequence=/BRCosmetics/Accessories/FORT_Backpacks/Backpack_M_MED_Alien_Flora/Animations/Alien_Flora_Idle_backpack
      blendIn {blendOption=Linear, blendTime=0.2, customCurve=null}
     }
     {
      chanceToPlay=0.3
      maxLoopCount=0
      maxPlayRate=1
      minLoopCount=0
      minPlayRate=1
      sequence=/BRCosmetics/Accessories/FORT_Backpacks/Backpack_M_MED_Alien_Flora/Animations/Alien_Flora_BiteBreak_backpack
      blendIn {blendOption=Linear, blendTime=0.2, customCurve=null}
     }
    ]
   }
   n2 = Get skydiving() -> (skydiving:bool)
   n3 = "Transform (Modify) Bone - Bone: tail_base"!(ComponentPose=Local To Component(LocalPose=@n1), bAlphaBoolEnabled:bool=@n2, Translation="0,0,0", Rotation="0,-20,0", Scale="1,1,1") -> (Pose) {
    alphaInputType=Bool
    rotationMode=BMM_Additive
    rotationSpace=BCS_BoneSpace
    alphaBoolBlend {blendInTime=1, blendOption=Linear}
    boneToModify {boneName=tail_base}
    rotation {yaw=-20}
   }
   n4 = Get NewVar_0() -> (NewVar_0:Rotator)
   n5 = "Transform (Modify) Bone - Bone: tail_01"!(ComponentPose=@n3, bAlphaBoolEnabled:bool=@n2, Translation="0,0,0", Rotation=@n4, Scale="1,1,1") -> (Pose) {
    alphaInputType=Bool
    rotationMode=BMM_Additive
    rotationSpace=BCS_BoneSpace
    alphaBoolBlend {blendInTime=1, blendOption=Linear}
    boneToModify {boneName=tail_01}
   }
   n6 = "Transform (Modify) Bone - Bone: tail_02"!(ComponentPose=@n5, bAlphaBoolEnabled:bool=@n2, Translation="0,0,0", Rotation=@n4, Scale="1,1,1") -> (Pose) {
    alphaInputType=Bool
    rotationMode=BMM_Additive
    rotationSpace=BCS_BoneSpace
    alphaBoolBlend {blendInTime=1, blendOption=Linear}
    boneToModify {boneName=tail_02}
   }
   n7 = "Transform (Modify) Bone - Bone: tail_03"!(ComponentPose=@n6, bAlphaBoolEnabled:bool=@n2, Translation="0,0,0", Rotation=@n4, Scale="1,1,1") -> (Pose) {
    alphaInputType=Bool
    rotationMode=BMM_Additive
    rotationSpace=BCS_BoneSpace
    alphaBoolBlend {blendInTime=1, blendOption=Linear}
    boneToModify {boneName=tail_03}
   }
   n8 = "Transform (Modify) Bone - Bone: tail_04"!(ComponentPose=@n7, bAlphaBoolEnabled:bool=@n2, Translation="0,0,0", Rotation=@n4, Scale="1,1,1") -> (Pose) {
    alphaInputType=Bool
    rotationMode=BMM_Additive
    rotationSpace=BCS_BoneSpace
    alphaBoolBlend {blendInTime=1, blendOption=Linear}
    boneToModify {boneName=tail_04}
   }
   n9 = "Transform (Modify) Bone - Bone: tail_05"!(ComponentPose=@n8, bAlphaBoolEnabled:bool=True, Translation="0,0,0", Rotation=@n4, Scale="1,1,1") -> (Pose) {
    alphaInputType=Bool
    rotationMode=BMM_Additive
    rotationSpace=BCS_BoneSpace
    alphaBoolBlend {blendInTime=1, blendOption=Linear}
    boneToModify {boneName=tail_05}
   }
   n10 = "Transform (Modify) Bone - Bone: tail_06"!(ComponentPose=@n9, bAlphaBoolEnabled:bool=@n2, Translation="0,0,0", Rotation=@n4, Scale="1,1,1") -> (Pose) {
    alphaInputType=Bool
    rotationMode=BMM_Additive
    rotationSpace=BCS_BoneSpace
    alphaBoolBlend {blendInTime=1, blendOption=Linear}
    boneToModify {boneName=tail_06}
   }
   n11 = "Transform (Modify) Bone - Bone: tail_07"!(ComponentPose=@n10, bAlphaBoolEnabled:bool=@n2, Translation="0,0,0", Rotation=@n4, Scale="1,1,1") -> (Pose) {
    alphaInputType=Bool
    rotationMode=BMM_Additive
    rotationSpace=BCS_BoneSpace
    alphaBoolBlend {blendInTime=1, blendOption=Linear}
    boneToModify {boneName=tail_07}
   }
   n12 = "Transform (Modify) Bone - Bone: tail_07"(ComponentPose=@n11, bAlphaBoolEnabled=@n2, Translation="0,0,0", Rotation=@n4, Scale="1,1,1") {
    alphaInputType=Bool
    rotationMode=BMM_Additive
    rotationSpace=BCS_BoneSpace
    alphaBoolBlend {blendInTime=1, blendOption=Linear}
    boneToModify {boneName=tail_07}
   }
   n13 = "Transform (Modify) Bone - Bone: tail_07"(ComponentPose=@n12, bAlphaBoolEnabled=@n2, Translation="0,0,0", Rotation=@n4, Scale="1,1,1") {
    alphaInputType=Bool
    rotationMode=BMM_Additive
    rotationSpace=BCS_BoneSpace
    alphaBoolBlend {blendInTime=1, blendOption=Linear}
    boneToModify {boneName=tail_07}
   }
   n14 = "Transform (Modify) Bone - Bone: tail_08"!(ComponentPose=@n13, bAlphaBoolEnabled:bool=@n2, Translation="0,0,0", Rotation=@n4, Scale="1,1,1") -> (Pose) {
    alphaInputType=Bool
    rotationMode=BMM_Additive
    rotationSpace=BCS_BoneSpace
    alphaBoolBlend {blendInTime=1, blendOption=Linear}
    boneToModify {boneName=tail_08}
   }
   n15 = "Transform (Modify) Bone - Bone: tail_09"!(ComponentPose=@n14, bAlphaBoolEnabled:bool=@n2, Translation="0,0,0", Rotation=@n4, Scale="1,1,1") -> (Pose) {
    alphaInputType=Bool
    rotationMode=BMM_Additive
    rotationSpace=BCS_BoneSpace
    alphaBoolBlend {blendInTime=1, blendOption=Linear}
    boneToModify {boneName=tail_09}
   }
   n16 = "Transform (Modify) Bone - Bone: tail_10"!(ComponentPose=@n15, bAlphaBoolEnabled:bool=@n2, Translation="0,0,0", Rotation=@n4, Scale="1,1,1") -> (Pose) {
    alphaInputType=Bool
    rotationMode=BMM_Additive
    rotationSpace=BCS_BoneSpace
    alphaBoolBlend {blendInTime=1, blendOption=Linear}
    boneToModify {boneName=tail_10}
   }
   n17 = "Transform (Modify) Bone - Bone: tail_11"!(ComponentPose=@n16, bAlphaBoolEnabled:bool=@n2, Translation="0,0,0", Rotation=@n4, Scale="1,1,1") -> (Pose) {
    alphaInputType=Bool
    rotationMode=BMM_Additive
    rotationSpace=BCS_BoneSpace
    alphaBoolBlend {blendInTime=1, blendOption=Linear}
    boneToModify {boneName=tail_11}
   }
   n18 = "Transform (Modify) Bone - Bone: tail_12"!(ComponentPose=@n17, bAlphaBoolEnabled:bool=@n2, Translation="0,0,0", Rotation=@n4, Scale="1,1,1") -> (Pose) {
    alphaInputType=Bool
    rotationMode=BMM_Additive
    rotationSpace=BCS_BoneSpace
    alphaBoolBlend {blendInTime=1, blendOption=Linear}
    boneToModify {boneName=tail_12}
   }
   n19 = "Transform (Modify) Bone - Bone: tail_13"!(ComponentPose=@n18, bAlphaBoolEnabled:bool=@n2, Translation="0,0,0", Rotation=@n4, Scale="1,1,1") -> (Pose) {
    alphaInputType=Bool
    rotationMode=BMM_Additive
    rotationSpace=BCS_BoneSpace
    alphaBoolBlend {blendInTime=1, blendOption=Linear}
    boneToModify {boneName=tail_13}
   }
   Output Pose(Result=Component To Local(ComponentPose=@n19)) {layerGroup=None}
  ]}
 EventGraph {nodes [
   Event BlueprintUpdateAnimation()
   > n1 = Branch(Condition=OR Boolean!(A=Get bIsSkydiveFloating(), B=false, C:bool=Get bIsSkydiveDiveMode()) -> (ReturnValue))
   > Set skydiving(skydiving:bool=true) -> (Output_Get:bool)
   > Set NewVar_0(NewVar_0:Rotator="0,-5,0") -> (Output_Get:Rotator)
   Set NewVar_0(execute=@n1.else, NewVar_0="0, 0, 0")
   TryGetPawnOwner()
   Get bIsParachuteOpened()
  ]}
]
variables [
 NewVar_0 {type=Rotator, default=<rot 0 -5 0>}
 skydiving {type=boolean}
]
